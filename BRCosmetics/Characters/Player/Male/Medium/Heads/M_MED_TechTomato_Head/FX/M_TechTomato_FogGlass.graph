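# Additive unlit material that draws breath-like fog on the glass of the TechTomato player head, usable on skeletal meshes and Niagara mesh particles.
# Two time-driven cycles, half a breath apart, erode a blur-noise texture in and out via sine and MF_Erosion.
# Vector noise of each cycle's time offsets the UVs to vary the pattern.
# The summed fog is Fresnel-masked, then tinted by FogColor and particle color.
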Material /BRCosmetics/Characters/Player/Male/Medium/Heads/M_MED_TechTomato_Head/FX/M_TechTomato_FogGlass
graph {
 properties {
  bUsedWithMeshDeformer=true
  bUsedWithNiagaraMeshParticles=true
  bUsedWithSkeletalMesh=true
  blendMode=BLEND_Additive
  pixelDepthOffsetMode=PDOM_Legacy
  shadingModel=MSM_Unlit
  translucencyPass=MTP_BeforeDOF
  usedShadingModels=Unlit
 }
 nodes [
  n1 = ParticleColor()
  n2 = VectorParameter() {parameterName=FogColor, defaultValue=<rgba 0.21 0.2655 0.3 0>}
  n3 = TextureCoordinate()
  n4 = Divide(A=Time(), B=ScalarParameter() {defaultValue=7, parameterName=Breath Duration})
  n5 = ComponentMask(None=VectorNoise(World Position=@n4)) {g=true, r=true}
  n6 = TextureObject() {samplerType=SAMPLERTYPE_Grayscale, texture=/BRCosmetics/Effects/Fort_Effects/Emotes/SwissKale/T_Blur_Noise}
  n7 = Power(
   Base=TextureSample(UVs=Add(A=@n3, B=@n5), Tex=@n6) {samplerType=SAMPLERTYPE_LinearGrayscale, texture=/Game/TimeOfDay/TODM_Sequence/Materials/Sky/Clouds/T_Asteria_Skybox_Clouds}.R
   Exp=ScalarParameter() {defaultValue=0.3, parameterName=Exponent}
  ) {constExponent=3}
  n8 = Add(A=@n4) {constB=0.75}
  n9 = OneMinus(None=ConstantBiasScale(None=Sine(None=@n8)))
  n10 = ScalarParameter() {defaultValue=3, parameterName=Fog Softness}
  n11 = TextureCoordinate()
  n12 = ScalarParameter() {defaultValue=7, parameterName=Breath Duration}
  n13 = Divide(A=Add(A=Time(), B=Multiply(A=@n12) {constB=1.5}) {constB=15}, B=@n12)
  n14 = ComponentMask(None=VectorNoise(World Position=@n13)) {g=true, r=true}
  n15 = Power(
   Base=TextureSample(UVs=Add(A=@n11, B=@n14), Tex=@n6) {samplerType=SAMPLERTYPE_LinearGrayscale, texture=&T_Asteria_Skybox_Clouds}.R
   Exp=ScalarParameter() {defaultValue=0.3, parameterName=Exponent}
  ) {constExponent=3}
  n16 = Add(A=@n13) {constB=0.75}
  n17 = OneMinus(None=ConstantBiasScale(None=Sine(None=@n16)))
  n18 = Multiply(
   A=Add(
    A=MaterialFunctionCall!(Gradient=@n7, Erosion Amount=@n9, Softness=@n10) {materialFunction=/BRCosmetics/Athena/Effects/Particles/Skydiving/NS_Contrail_DarkAndLight/Functions/MF_Erosion}.Default Range
    B=MaterialFunctionCall(Gradient=@n15, Erosion Amount=@n17, Softness=@n10) {materialFunction=&MF_Erosion}.Default Range
   )
   B=MaterialFunctionCall(Brightness=Constant() {r=1}, Exponent=ScalarParameter() {defaultValue=0.8, parameterName=Fresnel Exponent}) {materialFunction=&MF_Fresnel_FX}.Fresnel
  )
  MaterialFunctionCall(Emissive=Multiply(A=@n1.RGB, B=Multiply(A=@n2.RGB, B=@n18))) {materialFunction=&MF_FX_EmissiveAdjustments}
  Multiply(A=@n1.A, B=@n18)
 ]
}
