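# Idle Niagara effect for the TechTomato character head, adding glowing eyes and fog with a glass fog mesh.
# EyeGlow pins two flickering user-colored sprites to skeletal mesh sockets, scaling alpha by view angle and brightness by time of day.
# FogFlipbook loops SubUV puffs and FogGlass fades in a mesh by system age, both bone-attached and tinted by time of day.
# All three emitters set VisibilityTag from User.bisHeadHidden.
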NiagaraSystem /BRCosmetics/Characters/Player/Male/Medium/Heads/M_MED_TechTomato_Head/FX/NS_TechTomato_Idle
dependencies {
 usedDataInterfaces [
  &NiagaraDataInterfaceMeshRendererInfo
  &NiagaraDataInterfaceCamera
  &NiagaraDataInterfaceCurve
  &NiagaraDataInterfaceColorCurve
  &NiagaraDataInterfaceSkeletalMesh
 ]
 usedDynamicInputs [
  &Multiply_FloatByInt
  &SpawnGroup
  &Multiply_Vector
  &MakeVector
 ]
 usedModules [
  &self:SystemSpawnScript.SystemScriptSource.SystemScriptGraph.NiagaraNodeAssignment_0.SetVariables_#1
  &SystemState
 ]
}
emitters {
 EyeGlow {
  emitterUpdateScript [
   &EmitterState(
    Inactive Response=::NewEnumerator1
    Life Cycle Mode=::NewEnumerator1
    Loop Behavior=::NewEnumerator1
    Loop Duration Mode=::NewEnumerator1
    UseLoopDelay=false
   )
   &SpawnBurst_Instantaneous(Spawn Count=2, Use Spawn Probability=false, UseLoopCountLimit=false)
   /Niagara/Noise/Waveform(
    EnableExponent1=false
    GlobalClampMax=false
    GlobalClampMin=false
    UseInput1=false
    UsePhase1=false
    "Waveform [1]":ENiagara_Waveforms=::NewEnumerator12
    "[1] Frequency"=7
   )
   /SetVariables(Emitter.Flicker:NiagaraFloat=&Add_Float(A=&Multiply_Float(A=$Emitter.Waveform.Waveform, B=0.5), B=0.5)) {isSetParameters=true}
  ]
  particleSpawnScript [
   &PartitionParticles(ID to Partition:NiagaraInt32=$Engine.ExecIndex)
   &InitializeParticle(
    Color=$User.AdvancedEmissiveColor
    Lifetime=5
    Sprite Size Mode=::NewEnumerator3
    Uniform Sprite Size=5
    UsePositionOffset=true
    Write Material Random=true
   )
   &CameraOffset(Camera Offset Amount=2)
  ]
  particleUpdateScript [
   &ParticleState(Kill Particles When Lifetime Has Elapsed=false, Let Infinitely Lived Particles Die When Emitter Deactivates=true)
   &SkeletalMeshLocation(
    Bone / Socket Index=$Particles.PartitionParticles.Partition
    MaskSpawnGroup=false
    Mesh Sampling Type=::NewEnumerator5
    Skeletal Mesh=$System.Skeletal Mesh
    Socket Sampling Mode=::NewEnumerator5
    Use Sampled Mesh Attribute as Ribbon ID=false
   )
   &ScaleColor(
    Scale Alpha=&Multiply_Float(
     A=&Multiply_Float(A=$Emitter.Flicker, B=0.1)
     B=&DotProduct(
      Vector A=$Engine.Owner.SystemYAxis
      Vector B=&ReturnCameraVector(Camera Vectors=::NewEnumerator4, Coordinate Space=::Simulation, Origin Position=$Particles.Position)
     )
    )
    Scale RGB=&VectorFromFloat(Value=&FNTimeOfDayScalar(Night Scale=1, Morning Scale=1, Day Scale=1.5, Evening Scale=1))
    ScaleA=true
    ScaleRGB=true
   ) {module=ScaleColor001}
   /SetVariables(Particles.VisibilityTag=&MakeCustomIntFromBool(Bool=$User.bisHeadHidden, True INT=1, False INT=0)) {isSetParameters=true, module=SetVariables_#2}
  ]
  rendererClasses [&NiagaraSpriteRendererProperties]
  renderers [{rendererClass=&NiagaraSpriteRendererProperties}]
 }
 FogFlipbook {
  emitterUpdateScript [
   &EmitterState(Life Cycle Mode=::NewEnumerator1, Loop Duration=1, UseLoopDelay=true)
   &SpawnRate(SpawnRate=2, Use Spawn Probability=false) {module=SpawnRate001}
  ]
  particleSpawnScript [&InitializeParticle(
    Color Maximum=<rgba 0.54 0.6 0.6 0.4>
    Color Minimum=<rgba 0.95 0.9833 1 0.6>
    Color Mode=::NewEnumerator2
    Lifetime=&RandomRangeFloat(Minimum=1.25, Maximum=2.5)
    Position Mode=::NewEnumerator0
    Sprite Rotation Angle=&RandomRangeFloat(Minimum=0, Maximum=1, Evaluation Type=::NewEnumerator1)
    Sprite Rotation Mode=::NewEnumerator1
    Sprite Size=&Vector2DFromFloat(Value=&RandomRangeFloat(Minimum=24, Maximum=25, Evaluation Type=::NewEnumerator1))
    Sprite Size Mode=::NewEnumerator4
    Write Material Random=true
   )]
  particleUpdateScript [
   &ParticleState()
   &ScaleSpriteSize(
    Scale Sprite Size Mode=::NewEnumerator0
    Uniform Scale Factor=&FloatFromCurve(FloatCurve=DataInterface() {propertyValues {Curve {keys [
         {arriveTangent=1.5191, interpMode=RCIM_Cubic, leaveTangent=1.5191, tangentMode=RCTM_User, time=0, value=0.5}
         {interpMode=RCIM_Cubic, leaveTangent=0.4544, time=1, value=1}
        ]}}})
   )
   &ScaleColor(
    Scale Alpha=&FloatFromCurve(
     FloatCurve=DataInterface() {propertyValues {Curve {keys [
         {arriveTangent=5.1937, interpMode=RCIM_Cubic, leaveTangent=5.1937, tangentMode=RCTM_User, time=0, value=0}
         <at 0.5 1 RCIM_Cubic>
         {arriveTangent=-5.1244, interpMode=RCIM_Cubic, leaveTangent=-5.1244, tangentMode=RCTM_User, time=1, value=0}
        ]}}}
     Scale Curve=0.25
    )
    Scale RGB=&MakeVectorFromLinearColorRGB(Color=&ColorFromCurve(
      ColorCurve=DataInterface() {propertyValues {
        AlphaCurve {keys [<at 0 1>,<at 1 1>]}
        BlueCurve {keys [
          {interpMode=RCIM_Cubic, leaveTangentWeight=0.3333, tangentMode=RCTM_User, time=0.0249, value=1}
          {arriveTangentWeight=2.9077, interpMode=RCIM_Cubic, leaveTangentWeight=2.9077, tangentMode=RCTM_User, time=0.0976, value=0.7}
          {arriveTangentWeight=0.5797, interpMode=RCIM_Cubic, leaveTangentWeight=0.5797, tangentMode=RCTM_User, time=0.3272, value=0.9}
          {arriveTangentWeight=0.74, interpMode=RCIM_Cubic, leaveTangentWeight=0.74, tangentMode=RCTM_User, time=0.6455, value=0.9}
          {arriveTangentWeight=0.8064, interpMode=RCIM_Cubic, leaveTangentWeight=0.8064, tangentMode=RCTM_User, time=0.8135, value=0.8}
          {arriveTangentWeight=1.6709, interpMode=RCIM_Cubic, leaveTangentWeight=1.6709, tangentMode=RCTM_User, time=0.8751, value=0.8}
          {arriveTangentWeight=0.3333, interpMode=RCIM_Cubic, tangentMode=RCTM_User, time=0.9756, value=1}
         ]}
        GreenCurve {keys [
          {interpMode=RCIM_Cubic, leaveTangentWeight=0.3333, tangentMode=RCTM_User, time=0.0249, value=0.8553}
          {arriveTangentWeight=0.4863, interpMode=RCIM_Cubic, leaveTangentWeight=0.4863, tangentMode=RCTM_User, time=0.0976, value=0.7726}
          {arriveTangentWeight=0.5565, interpMode=RCIM_Cubic, leaveTangentWeight=0.5565, tangentMode=RCTM_User, time=0.3272, value=0.965}
          {arriveTangentWeight=0.8775, interpMode=RCIM_Cubic, leaveTangentWeight=0.8775, tangentMode=RCTM_User, time=0.6455, value=0.965}
          {arriveTangentWeight=0.9618, interpMode=RCIM_Cubic, leaveTangentWeight=0.9618, tangentMode=RCTM_User, time=0.8135, value=0.8088}
          {arriveTangentWeight=0.6147, interpMode=RCIM_Cubic, leaveTangentWeight=0.6147, tangentMode=RCTM_User, time=0.8751, value=0.8088}
          {arriveTangentWeight=0.3333, interpMode=RCIM_Cubic, tangentMode=RCTM_User, time=0.9756, value=0.8553}
         ]}
        RedCurve {keys [
          {interpMode=RCIM_Cubic, leaveTangentWeight=0.3333, tangentMode=RCTM_User, time=0.0249, value=0.8}
          {arriveTangentWeight=2.5584, interpMode=RCIM_Cubic, leaveTangentWeight=2.5584, tangentMode=RCTM_User, time=0.0976, value=1}
          {arriveTangentWeight=0.3333, interpMode=RCIM_Cubic, leaveTangentWeight=0.3333, tangentMode=RCTM_User, time=0.3272, value=1}
          {arriveTangentWeight=0.3333, interpMode=RCIM_Cubic, leaveTangentWeight=0.3333, tangentMode=RCTM_User, time=0.6455, value=1}
          {arriveTangentWeight=0.3333, interpMode=RCIM_Cubic, leaveTangentWeight=0.3333, tangentMode=RCTM_User, time=0.8135, value=1}
          {arriveTangentWeight=1.4874, interpMode=RCIM_Cubic, leaveTangentWeight=1.4874, tangentMode=RCTM_User, time=0.8751, value=1}
          {arriveTangentWeight=0.3333, interpMode=RCIM_Cubic, tangentMode=RCTM_User, time=0.9756, value=0.8}
         ]}
       }}
      CurveIndex=Linked() {linkedVariable {name=NPC.FortniteNPC.RemappedTimeOfDay_0-1, type=NiagaraFloat}}
     ))
    ScaleA=true
    ScaleRGB=true
   )
   /Niagara/Modules/Update/SubUV/SubUVAnimation(Number Of Frames=64)
   &SkeletalMeshLocation(
    Bone / Socket Index=1
    Bone Sampling Mode=::NewEnumerator4
    MaskSpawnGroup=false
    Sampled Position Offset=<12 0 0>
    Skeletal Mesh=$System.Skeletal Mesh
    Use Sampled Mesh Attribute as Ribbon ID=false
   )
   &CameraOffset(Camera Offset Amount=5)
   /SetVariables(Particles.VisibilityTag=&MakeCustomIntFromBool(Bool=$User.bisHeadHidden, True INT=1, False INT=0)) {isSetParameters=true, module=SetVariables_#2}
  ]
  rendererClasses [&NiagaraSpriteRendererProperties]
  renderers [{rendererClass=&NiagaraSpriteRendererProperties}]
 }
 FogGlass {
  emitterUpdateScript [
   &EmitterState(
    Inactive Response=::NewEnumerator1
    Life Cycle Mode=::NewEnumerator1
    Loop Behavior=::NewEnumerator1
    Loop Duration Mode=::NewEnumerator1
    UseLoopDelay=false
   )
   &SpawnBurst_Instantaneous(Use Spawn Probability=false, UseLoopCountLimit=false)
  ]
  particleSpawnScript [
   &InitializeParticle(
    Lifetime=5
    Mesh Scale=&Multiply_VectorByFloat(Vector=<0.5 0.5 0.42>, Float=0.65)
    Mesh Scale Mode=::NewEnumerator4
    Sprite Size Mode=::NewEnumerator3
    Uniform Sprite Size=5
    UsePositionOffset=true
    Write Material Random=true
   )
   &InitialMeshOrientation(Mesh Orientation Mode=::NewEnumerator7, Rotation=<0 0.25 -0.05>, Use Rotation Vector=true)
  ]
  particleUpdateScript [
   &ParticleState(Kill Particles When Lifetime Has Elapsed=false, Let Infinitely Lived Particles Die When Emitter Deactivates=true)
   &ScaleColor(
    Scale Alpha=&FloatFromCurve(
     FloatCurve=DataInterface() {propertyValues {Curve {keys [
         {arriveTangent=4.1785, interpMode=RCIM_Cubic, leaveTangent=4.1785, tangentMode=RCTM_User, time=0, value=0}
         {arriveTangent=1.1179, interpMode=RCIM_Cubic, leaveTangent=1.1179, tangentMode=RCTM_User, time=0.2042, value=0.6228}
         {arriveTangent=0.0224, interpMode=RCIM_Cubic, leaveTangent=0.0224, tangentMode=RCTM_User, time=1, value=1}
        ]}}}
     CurveIndex=$System.Age
     Scale Curve=15
    )
    Scale RGB=&MakeVectorFromLinearColorRGB(Color=&ColorFromCurve(
      ColorCurve=DataInterface() {propertyValues {
        AlphaCurve {keys [<at 0 1>,<at 1 1>]}
        BlueCurve {keys [
          {interpMode=RCIM_Cubic, leaveTangentWeight=0.3333, tangentMode=RCTM_User, time=0.0249, value=1}
          {arriveTangentWeight=2.9077, interpMode=RCIM_Cubic, leaveTangentWeight=2.9077, tangentMode=RCTM_User, time=0.0976, value=0.7}
          {arriveTangentWeight=0.5797, interpMode=RCIM_Cubic, leaveTangentWeight=0.5797, tangentMode=RCTM_User, time=0.3272, value=0.7298}
          {arriveTangentWeight=0.74, interpMode=RCIM_Cubic, leaveTangentWeight=0.74, tangentMode=RCTM_User, time=0.6455, value=0.7298}
          {arriveTangentWeight=0.8064, interpMode=RCIM_Cubic, leaveTangentWeight=0.8064, tangentMode=RCTM_User, time=0.8135, value=0.8}
          {arriveTangentWeight=1.6709, interpMode=RCIM_Cubic, leaveTangentWeight=1.6709, tangentMode=RCTM_User, time=0.8751, value=0.8}
          {arriveTangentWeight=0.3333, interpMode=RCIM_Cubic, tangentMode=RCTM_User, time=0.9756, value=1}
         ]}
        GreenCurve {keys [
          {interpMode=RCIM_Cubic, leaveTangentWeight=0.3333, tangentMode=RCTM_User, time=0.0249, value=0.8553}
          {arriveTangentWeight=0.4863, interpMode=RCIM_Cubic, leaveTangentWeight=0.4863, tangentMode=RCTM_User, time=0.0976, value=0.7726}
          {arriveTangentWeight=0.5565, interpMode=RCIM_Cubic, leaveTangentWeight=0.5565, tangentMode=RCTM_User, time=0.3272, value=0.9054}
          {arriveTangentWeight=0.8775, interpMode=RCIM_Cubic, leaveTangentWeight=0.8775, tangentMode=RCTM_User, time=0.6455, value=0.9054}
          {arriveTangentWeight=0.9618, interpMode=RCIM_Cubic, leaveTangentWeight=0.9618, tangentMode=RCTM_User, time=0.8135, value=0.8088}
          {arriveTangentWeight=0.6147, interpMode=RCIM_Cubic, leaveTangentWeight=0.6147, tangentMode=RCTM_User, time=0.8751, value=0.8088}
          {arriveTangentWeight=0.3333, interpMode=RCIM_Cubic, tangentMode=RCTM_User, time=0.9756, value=0.8553}
         ]}
        RedCurve {keys [
          {interpMode=RCIM_Cubic, leaveTangentWeight=0.3333, tangentMode=RCTM_User, time=0.0249, value=0.8}
          {arriveTangentWeight=2.5584, interpMode=RCIM_Cubic, leaveTangentWeight=2.5584, tangentMode=RCTM_User, time=0.0976, value=1}
          {arriveTangentWeight=0.3333, interpMode=RCIM_Cubic, leaveTangentWeight=0.3333, tangentMode=RCTM_User, time=0.3272, value=1}
          {arriveTangentWeight=0.3333, interpMode=RCIM_Cubic, leaveTangentWeight=0.3333, tangentMode=RCTM_User, time=0.6455, value=1}
          {arriveTangentWeight=0.3333, interpMode=RCIM_Cubic, leaveTangentWeight=0.3333, tangentMode=RCTM_User, time=0.8135, value=1}
          {arriveTangentWeight=1.4874, interpMode=RCIM_Cubic, leaveTangentWeight=1.4874, tangentMode=RCTM_User, time=0.8751, value=1}
          {arriveTangentWeight=0.3333, interpMode=RCIM_Cubic, tangentMode=RCTM_User, time=0.9756, value=0.8}
         ]}
       }}
      CurveIndex=Linked() {linkedVariable {name=NPC.FortniteNPC.RemappedTimeOfDay_0-1, type=NiagaraFloat}}
     ))
    ScaleA=true
    ScaleRGB=true
   ) {module=ScaleColor001}
   &SkeletalMeshLocation(
    Bone / Socket Index=1
    Bone Sampling Mode=::NewEnumerator4
    MaskSpawnGroup=false
    Sampled Position Offset=<12 0 0>
    Skeletal Mesh=$System.Skeletal Mesh
    Use Sampled Mesh Attribute as Ribbon ID=false
   )
   /SetVariables(Particles.VisibilityTag=&MakeCustomIntFromBool(Bool=$User.bisHeadHidden, True INT=1, False INT=0)) {isSetParameters=true, module=SetVariables_#2}
  ]
  rendererClasses [&NiagaraMeshRendererProperties]
  renderers [{rendererClass=&NiagaraMeshRendererProperties, rendererIndex=0}]
 }
}
summary {
 systemName=NS_TechTomato_Idle
 emitters [
  {emitterName=EyeGlow, rendererClasses [&NiagaraSpriteRendererProperties]}
  {emitterName=FogFlipbook, rendererClasses [&NiagaraSpriteRendererProperties]}
  {emitterName=FogGlass, rendererClasses [&NiagaraMeshRendererProperties]}
 ]
 userVariables [
  User.AdvancedEmissiveColor {type=LinearColor, description="", defaultValue {struct=LinearColor, value=<rgba 0.7565 2.6102 3 1>}}
  User.bisHeadHidden {type=NiagaraBool}
 ]
}
